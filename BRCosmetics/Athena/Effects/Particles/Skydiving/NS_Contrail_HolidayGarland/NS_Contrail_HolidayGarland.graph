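# Skydiving contrail cosmetic that trails a holiday garland with bows, lights, glass ornaments, glowing stars and pine bits from skeletal mesh sockets.
# Garland and bow meshes follow sockets, orient to the owner and feed noise settings to their materials.
# Random-colored sprites with light renderers pulse along, while shed pieces scale lifetime or velocity with PlayerSpeed.
# User.V2 sets garland, bow and pine counts, V3 gates ornament and star bursts, and front-end modifiers alter debris velocity.
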NiagaraSystem /BRCosmetics/Athena/Effects/Particles/Skydiving/NS_Contrail_HolidayGarland/NS_Contrail_HolidayGarland
dependencies {usedDataInterfaces [&NiagaraDataInterfaceCurve], usedDynamicInputs [], usedModules [&SystemState]}
emitters {
 GlassOrnaments {
  emitterSpawnScript [&FN_PreserveSkydiveUserVariables() {module=PreserveUserVariables}]
  emitterUpdateScript [
   &EmitterState(Life Cycle Mode=::NewEnumerator1, Loop Duration=1, UseLoopDelay=false)
   &SpawnBurst_Instantaneous(
    Spawn Count=&UniformRangedInt(Minimum=0, Maximum=2)
    Spawn Probability=&Lerp_Float(A=0, B=1, Alpha=$User.V3)
    Use Spawn Probability=true
   )
  ]
  particleSpawnScript [
   &FN_SetSpawnGroup_v2(Number of Spawn Groups=4)
   /Niagara/Modules/Spawn/Initialization/InitializeParticle(
    Lifetime=&Lerp_Float(A=0.5, B=1, Alpha=$User.PlayerSpeed)
    Mesh Scale=&VectorFromFloat(Value=&UniformRangedFloat(Minimum=1, Maximum=1.4))
    Module.Write Color=true
    Module.Write Lifetime=true
    Module.Write Mass=true
    Module.Write Material Random=true
    Module.Write Position=false
    Module.Write Scale=true
    Module.Write SpriteRotation=false
    Module.Write SpriteSize=false
    Position=&SimulationPosition()
   )
   &SkeletalMeshLocation(
    Bone / Socket Index=&SpawnGroup()
    Bone Sampling Mode=::NewEnumerator4
    Inherit Velocity=false
    MaskSpawnGroup=false
    Skeletal Mesh=$User.SkeletalMesh
   )
   /Niagara/Modules/Spawn/Orientation/V2/InitialMeshOrientation(
    Rotation=&VectorFromFloat(Value=&UniformRangedFloat(Minimum=-1, Maximum=1))
    Use Orientation Vector=true
    Use Rotation Vector=true
   )
   &AddVelocityInCone(
    Cone Angle=5
    Cone Axis=<0 0 1>
    Module.Use Velocity Falloff On Cone Axis=false
    Velocity Strength=&Lerp_Float(
     A=&FN_FrontEndFloatModifier(Gameplay=-900, Front End=250)
     B=&FN_FrontEndFloatModifier(Gameplay=-1250, Front End=125)
     Alpha=$User.PlayerSpeed
    )
   )
  ]
  particleUpdateScript [
   &ParticleState()
   &Drag(Drag=0.5, Use Linear Drag=true, Use Rotational Drag=false)
   &SolveForcesAndVelocity(Clamp Velocity=false, Limit Acceleration=false)
   &ScaleMeshSize(Scale Factor=&VectorFromFloat(Value=&FloatFromCurve(FloatCurve=DataInterface() {propertyValues {Curve {keys [<at 0 0>,<at 0.1 1>,<at 0.8353 1>,<at 1.0011 0.002>]}}})))
   /Niagara/Modules/Update/Orientation/V2/UpdateMeshOrientation(
    Rotation Rate:NiagaraFloat=&FloatFromCurve(FloatCurve=DataInterface() {propertyValues {Curve {keys [<at -0.0046 1.0113>,<at 1.0011 0.002>]}}})
    Rotation Vector:Vector3f=<1 1 1>
   )
  ]
  rendererClasses [&NiagaraMeshRendererProperties]
  renderers [{rendererClass=&NiagaraMeshRendererProperties, rendererIndex=0}]
 }
 GlowyStars {
  emitterSpawnScript [&FN_PreserveSkydiveUserVariables() {module=PreserveUserVariables}]
  emitterUpdateScript [
   &EmitterState(Life Cycle Mode=::NewEnumerator1, Loop Duration=0, UseLoopDelay=false)
   &SpawnBurst_Instantaneous(Spawn Count=4, Spawn Probability=&Lerp_Float(A=0, B=1, Alpha=$User.V3), Use Spawn Probability=true)
  ]
  particleSpawnScript [
   &FN_SetSpawnGroup_v2(Number of Spawn Groups=4) {module=FN_SetSpawnGroup_v2001}
   &InitializeParticle(
    Color=&MakeLinearColorFromVectorAndFloat(
     "Vector (RGB)"=&Lerp_Vector(
      A=&MakeVectorFromLinearColorRGB(Color=<rgba 0 0.1486 1 1>)
      B=&MakeVectorFromLinearColorRGB(Color=<rgba 1 0 0.0874 1>)
      Alpha=&UniformAOrBFloat(A=0)
     )
     "Float (Alpha)"=&UniformRangedFloat(Minimum=1, Maximum=1)
    )
    Lifetime=&Lerp_Float(A=&UniformRangedFloat(Minimum=0.5, Maximum=1.5), B=0.25, Alpha=$User.PlayerSpeed)
    Module.Write Color=true
    Module.Write Lifetime=true
    Module.Write Mass=false
    Module.Write Material Random=true
    Module.Write Position=true
    Module.Write Scale=false
    Module.Write SpriteRotation=false
    Module.Write SpriteSize=true
    Position=&SimulationPosition()
    Sprite Size=&Vector2DFromFloat(Value=&UniformRangedFloat(Minimum=8, Maximum=35))
   )
   &SkeletalMeshLocation(
    Bone / Socket Index=&SpawnGroup()
    Bone Sampling Mode=::NewEnumerator4
    Inherit Velocity=false
    MaskSpawnGroup=false
    Skeletal Mesh=$User.SkeletalMesh
   )
   /Niagara/Modules/Spawn/Location/SphereLocation(MaskSpawnGroup=false, Module.Override Randomness=false, Sphere Radius=10, Surface Only=false)
   &AddVelocity(
    Coordinate Space=::World
    Scale Added Velocity=&Lerp_Vector(A=<60 60 256>, B=<5 5 1500>, Alpha=$User.PlayerSpeed)
    Velocity=/Game/Effects/Niagara/DynamicInputs/Random/FN_RandomVectorInCone(
     Cone Angle:NiagaraFloat=80
     Cone Axis:Vector3f=<0 0 1>
     Cone Axis Coordinate Space:ENiagaraCoordinateSpace=::World
     Velocity Distribution Along Cone Axis:NiagaraFloat=0
    )
   )
  ]
  particleUpdateScript [
   &ParticleState()
   &Drag(Use Linear Drag=true, Use Rotational Drag=false)
   &SolveForcesAndVelocity(Clamp Velocity=false, Limit Acceleration=false)
   &ScaleColor(
    Scale Alpha=&Multiply_Float(
     A=&FloatFromCurve(FloatCurve=DataInterface() {propertyValues {Curve {keys [<at 0 0>,<at 0.2 1>,<at 1.0011 0.002>]}}})
     B=1
    )
    ScaleA=true
    ScaleRGB=true
   )
   &CameraOffset()
   &ScaleSpriteSize(Scale Factor=&Vector2DFromFloat(Value=&FloatFromCurve(FloatCurve=DataInterface() {propertyValues {Curve {keys [
          <at 0 0>
          {arriveTangent=-1.6664, interpMode=RCIM_Cubic, leaveTangent=-1.6664, tangentMode=RCTM_User, time=0.0663, value=1.0113}
          <at 1.0011 0.002>
         ]}}})))
  ]
  rendererClasses [&NiagaraSpriteRendererProperties]
  renderers [{rendererClass=&NiagaraSpriteRendererProperties}]
 }
 GreenGarland {
  emitterSpawnScript [&FN_PreserveSkydiveUserVariables() {module=PreserveUserVariables}]
  emitterUpdateScript [
   &EmitterState()
   &SpawnBurst_Instantaneous(
    Spawn Count=&MakeIntFromFloat(Float=&Lerp_Float(A=2, B=4, Alpha=$User.V2))
    Use Spawn Probability=false
   )
  ]
  particleSpawnScript [
   &FN_SetSpawnGroup_v2(Number of Spawn Groups:NiagaraInt32=&MakeIntFromFloat(Float=&Lerp_Float(A=2, B=4, Alpha=$User.V2)))
   &InitializeParticle(
    Lifetime=100
    Mesh Scale=&MakeVector(X=1, Y=1, Z=HlslExpression() {hlslExpression="Particles.SpawnGroup < 2 ?  1.0: 1.0"})
    Module.Write Color=true
    Module.Write Lifetime=false
    Module.Write Mass=true
    Module.Write Material Random=true
    Module.Write Position=false
    Module.Write Scale=false
    Module.Write SpriteRotation=true
    Module.Write SpriteSize=true
    Position=&SimulationPosition()
    Sprite Rotation=&UniformRangedFloat(Minimum=0, Maximum=360)
   )
  ]
  particleUpdateScript [
   /Game/Effects/Niagara/Modules/Skydiving/FN_DivingAttributes/FN_DivingAttributes_v2()
   &ParticleState(Kill Particles When Lifetime Has Elapsed=false, Let Infinitely Lived Particles Die When Emitter Deactivates=true)
   &SkeletalMeshLocation(
    Bone / Socket Index=&SpawnGroup()
    Bone Sampling Mode=::NewEnumerator4
    Inherit Velocity=false
    MaskSpawnGroup=false
    Skeletal Mesh=$User.SkeletalMesh
   )
   /BRCosmetics/Athena/Effects/Particles/Skydiving/NS_Contrail_AmmoBelt/Module/FN_OrientMeshToEngineOwner(
    Direction:Vector3f=&FN_FrontEndVector3Modifier(Gameplay=<0 0 -1>, Front End=<0 0 -1>)
    Set Mesh Orientation=true
    Vector:Vector3f=&TransformVector(Vector=$Engine.Owner.SystemYAxis, Source Space=::World, Destination Space=::Simulation, Apply Scale=true)
   )
   /Game/Effects/Niagara/Modules/Materials/FN_MeshMaterial(
    Amplitude=40
    Falloff=60
    Frequency=500
    Jitter Multiplier=1.8545
    Offset:NiagaraFloat=$Particles.AccumulatedNormalizedDive
    Scale Noise X Axis=2
    Scale Noise Y Axis=2
    Scale Noise Z Axis=0.8
    Speed=0.6
   )
   /SetVariables(Particles.MeshOrientation=&MultiplyQuaternion(
     Quaternion A=$Particles.MeshOrientation
     Quaternion B=&MakeQuatFromAxisAngle(Axis=<0 0 1>, AngleInDegrees=&Multiply_Float(A=$Particles.Initial.SpriteRotation, B=1))
    )) {isSetParameters=true}
   /SetVariables(Particles.Vector 4:Vector4f=<1 0 0 0>) {isSetParameters=true}
  ]
  rendererClasses [&NiagaraMeshRendererProperties]
  renderers [
   {rendererClass=&NiagaraMeshRendererProperties, rendererIndex=0}
   {rendererClass=&NiagaraMeshRendererProperties, rendererIndex=1}
   {rendererClass=&NiagaraMeshRendererProperties, rendererIndex=2}
   {rendererClass=&NiagaraMeshRendererProperties, rendererIndex=3}
  ]
 }
 Lights {
  emitterSpawnScript [&FN_PreserveSkydiveUserVariables() {module=PreserveUserVariables}]
  emitterUpdateScript [
   &EmitterState()
   &SpawnRate(SpawnRate=20, Use Spawn Probability=false)
   &SpawnBurst_Instantaneous(Spawn Count=4, Spawn Probability=&Lerp_Float(A=0, B=1, Alpha=$User.V2), Use Spawn Probability=true) {enabled=false}
  ]
  particleSpawnScript [
   &InitializeParticle(
    Color=&MakeLinearColor(
     R:NiagaraFloat=&UniformRangedFloat(Minimum=0, Maximum=6)
     G:NiagaraFloat=&UniformRangedFloat(Minimum=0, Maximum=6)
     B:NiagaraFloat=&UniformRangedFloat(Minimum=0, Maximum=6)
     A:NiagaraFloat=1
    )
    Lifetime=1.1
    Module.Write Color=true
    Module.Write Lifetime=true
    Module.Write Mass=true
    Module.Write Material Random=true
    Module.Write Position=false
    Module.Write Scale=false
    Module.Write SpriteRotation=false
    Module.Write SpriteSize=false
    Position=&SimulationPosition()
   )
   &FN_SetSpawnGroup_v2(Number of Spawn Groups=4) {enabled=false, module=FN_SetSpawnGroup_v2001}
   &PartitionParticles(Number Of Partitions=4)
  ]
  particleUpdateScript [
   &ParticleState(Kill Particles When Lifetime Has Elapsed=false, Let Infinitely Lived Particles Die When Emitter Deactivates=true)
   &SkeletalMeshLocation(
    Bone / Socket Index=$Particles.PartitionParticles.Partition
    Bone Sampling Mode=::NewEnumerator4
    MaskSpawnGroup=false
    Sampled Position Offset=<0 0 20>
    Sampled Position Offset Coordinate Space:ENiagaraSkelMeshTransforms=::NewEnumerator2
    Skeletal Mesh=$User.SkeletalMesh
   )
   /SetVariables(Particles.Position=&Add_Vector(A=Linked() {linkedVariable {name=Particles.Position, type=NiagaraPosition}}, B=<0 0 20>)) {enabled=false, isSetParameters=true}
   &ScaleColor(
    Scale RGB=&Multiply_Vector(
     A=<1 1 1>
     B=&VectorFromFloat(Value=&FloatFromCurve(
       FloatCurve=DataInterface() {propertyValues {Curve {keys [
           {arriveTangent=3.9421, interpMode=RCIM_Cubic, leaveTangent=3.9421, tangentMode=RCTM_User, time=0, value=0}
           <at 0.5 1 RCIM_Cubic>
           {arriveTangent=-3.6716, interpMode=RCIM_Cubic, leaveTangent=-3.6716, tangentMode=RCTM_User, time=1, value=0}
          ]}}}
       Scale Curve=5
      ))
    )
    ScaleA=true
    ScaleRGB=true
   )
   &FN_Light_Attributes(Create Exponent Attribute=true, Create Radius Attribute=true, Exponent=1, Radius=35)
  ]
  rendererClasses [&NiagaraLightRendererProperties, &NiagaraSpriteRendererProperties]
  renderers [
   {rendererClass=&NiagaraLightRendererProperties, rendererIndex=0}
   {rendererClass=&NiagaraSpriteRendererProperties, rendererIndex=1}
  ]
 }
 PineBits {
  emitterSpawnScript [&FN_PreserveSkydiveUserVariables() {module=PreserveUserVariables}]
  emitterUpdateScript [
   &EmitterState(Life Cycle Mode=::NewEnumerator1, Loop Duration=1, UseLoopDelay=false)
   &SpawnBurst_Instantaneous(
    Spawn Count=&MakeIntFromFloat(Float=&Lerp_Float(A=2, B=4, Alpha=$User.V2))
    Use Spawn Probability=false
   )
  ]
  particleSpawnScript [
   &FN_SetSpawnGroup_v2(Number of Spawn Groups:NiagaraInt32=&MakeIntFromFloat(Float=&Lerp_Float(A=2, B=4, Alpha=$User.V2))) {module=FN_SetSpawnGroup_v2001}
   &InitializeParticle(
    Lifetime=&Lerp_Float(
     A=&UniformRangedFloat(Minimum=1.4, Maximum=2)
     B=&UniformRangedFloat(Minimum=0.2, Maximum=0.325)
     Alpha=$User.PlayerSpeed
    )
    Mesh Scale=&UniformRangedVector(Minimum=<0.75 0.75 0.75>, Maximum=<1 1 1>)
    Module.Write Color=true
    Module.Write Lifetime=true
    Module.Write Mass=true
    Module.Write Material Random=true
    Module.Write Position=false
    Module.Write Scale=true
    Module.Write SpriteRotation=false
    Module.Write SpriteSize=false
    Position=&SimulationPosition()
   )
   &SkeletalMeshLocation(
    Bone / Socket Index=&SpawnGroup()
    Bone Sampling Mode=::NewEnumerator4
    Inherit Velocity=false
    MaskSpawnGroup=false
    Skeletal Mesh=$User.SkeletalMesh
   )
   &InitialMeshOrientation(
    Rotation=&VectorFromFloat(Value=&UniformRangedFloat(Minimum=-1, Maximum=1))
    Use Orientation Vector=true
    Use Rotation Vector=true
   )
   &AddVelocityInCone(
    Cone Angle=5
    Cone Axis=<0 0 1>
    Module.Use Velocity Falloff On Cone Axis=false
    Velocity Strength=&Lerp_Float(
     A=&FN_FrontEndFloatModifier(Gameplay=-1100, Front End=&UniformRangedFloat(Minimum=118, Maximum=350))
     B=&FN_FrontEndFloatModifier(Gameplay=-1250, Front End=&UniformRangedFloat(Minimum=118, Maximum=350))
     Alpha=$User.PlayerSpeed
    )
   )
  ]
  particleUpdateScript [
   &ParticleState()
   &Drag(Drag=0.5, Use Linear Drag=true, Use Rotational Drag=false)
   &SolveForcesAndVelocity(Clamp Velocity=false, Limit Acceleration=false)
   &ScaleMeshSize(Scale Factor=&VectorFromFloat(Value=&FloatFromCurve(FloatCurve=DataInterface() {propertyValues {Curve {keys [<at 0 0>,<at 0.1066 0.9779>]}}})))
   &UpdateMeshOrientation(Rotation Rate=2, Rotation Vector:Vector3f=<1 1 1>)
   &ScaleColor(
    Scale Alpha=&FloatFromCurve(FloatCurve=DataInterface() {propertyValues {Curve {keys [<at -0.0046 1.0113>,<at 1.0011 0.002>]}}})
    ScaleA=true
    ScaleRGB=true
   )
  ]
  rendererClasses [&NiagaraMeshRendererProperties]
  renderers [{rendererClass=&NiagaraMeshRendererProperties, rendererIndex=0}]
 }
 RibbonBows {
  emitterSpawnScript [&FN_PreserveSkydiveUserVariables() {module=PreserveUserVariables}]
  emitterUpdateScript [
   &EmitterState()
   &SpawnBurst_Instantaneous(
    Spawn Count=&MakeIntFromFloat(Float=&Lerp_Float(A=2, B=4, Alpha=$User.V2))
    Use Spawn Probability=false
   )
  ]
  particleSpawnScript [
   &FN_SetSpawnGroup_v2(Number of Spawn Groups:NiagaraInt32=&MakeIntFromFloat(Float=&Lerp_Float(A=2, B=4, Alpha=$User.V2))) {module=FN_SetSpawnGroup_v2001}
   &InitializeParticle(
    Lifetime=100
    Mesh Scale=&MakeVector(X=0.7, Y=0.7, Z=0.65)
    Module.Write Color=true
    Module.Write Lifetime=true
    Module.Write Mass=true
    Module.Write Material Random=true
    Module.Write Position=false
    Module.Write Scale=true
    Module.Write SpriteRotation=false
    Module.Write SpriteSize=true
    Position=&SimulationPosition()
    Sprite Size=<50 50>
   )
  ]
  particleUpdateScript [
   &ParticleState(Kill Particles When Lifetime Has Elapsed=false, Let Infinitely Lived Particles Die When Emitter Deactivates=true)
   &SkeletalMeshLocation(
    Bone / Socket Index=&SpawnGroup()
    Bone Sampling Mode=::NewEnumerator4
    Inherit Velocity=false
    MaskSpawnGroup=false
    Skeletal Mesh=$User.SkeletalMesh
   )
   &FN_OrientMeshToEngineOwner(
    Direction:Vector3f=&FN_FrontEndVector3Modifier(Gameplay=<0 0 -1>, Front End=<0 0 -1>)
    Mesh Forward Vector:Vector3f=<-1 0 0>
    Mesh Up Vector:Vector3f=<0 0.5 0>
    Set Mesh Orientation=true
    Vector:Vector3f=&TransformVector(Vector=$Engine.Owner.SystemYAxis, Source Space=::World, Destination Space=::Local)
   )
   &FN_MeshMaterial(Amplitude=25, Falloff=12, Frequency=5000, Jitter Multiplier=1.8545, Offset:NiagaraFloat=$Particles.AccumulatedNormalizedDive, Speed=3)
  ]
  rendererClasses [&NiagaraMeshRendererProperties]
  renderers [{rendererClass=&NiagaraMeshRendererProperties, rendererIndex=0}]
 }
}
summary {
 systemName=NS_Contrail_HolidayGarland
 emitters [
  {emitterName=GreenGarland, rendererClasses [&NiagaraMeshRendererProperties]}
  {emitterName=RibbonBows, rendererClasses [&NiagaraMeshRendererProperties]}
  {emitterName=GlassOrnaments, rendererClasses [&NiagaraMeshRendererProperties]}
  {emitterName=GlowyStars, rendererClasses [&NiagaraSpriteRendererProperties]}
  {emitterName=Lights, rendererClasses [&NiagaraLightRendererProperties, &NiagaraSpriteRendererProperties]}
  {emitterName=PineBits, rendererClasses [&NiagaraMeshRendererProperties]}
 ]
 userVariables [
  User.bIsFrontEnd {type=NiagaraBool}
  User.bIsFrontEndPreview {type=NiagaraBool}
  User.LightColor {type=LinearColor, description="", defaultValue {struct=LinearColor, value=<rgba 1 1 1 1>}}
  User.PlayerSpeed {type=NiagaraFloat, defaultValue=0, description=""}
  User.SkeletalMesh {
   type=NiagaraDataInterfaceSkeletalMesh
   defaultValue {
    struct=NiagaraExt_VariableValue_DataInterface
    value {dataInterface=&self:NiagaraDataInterfaceSkeletalMesh_0}
   }
  }
  User.V2 {type=NiagaraFloat, defaultValue=0, description=""}
  User.V3 {type=NiagaraFloat, defaultValue=0, description=""}
 ]
}
